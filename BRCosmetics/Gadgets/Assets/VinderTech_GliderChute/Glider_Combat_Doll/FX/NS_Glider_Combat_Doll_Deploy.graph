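# Niagara deploy effect for the Combat Doll glider cosmetic, drawing thruster-like flame meshes, blue glow and spark sprites, and a flickering light.
# Emitters sample sockets on the user skeletal mesh and fade or shrink over normalized loop age.
# User ForwardDot enlarges glow and alters flame lifetime and brightness with front-end overrides, while system burst values trigger extra flames and stretch them.
# Smoke and original Sparks emitters are disabled, leaving Sparks001 active.
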NiagaraSystem /BRCosmetics/Gadgets/Assets/VinderTech_GliderChute/Glider_Combat_Doll/FX/NS_Glider_Combat_Doll_Deploy
dependencies {
 usedDataInterfaces [
  &NiagaraDataInterfaceCurve
  &NiagaraDataInterfaceColorCurve
  &NiagaraDataInterfaceVector2DCurve
  &NiagaraDataInterfaceVectorCurve
  &NiagaraDataInterfaceCamera
 ]
 usedDynamicInputs [&Boolean_AndOperation, &Add_Float]
 usedModules [
  &self:SystemSpawnScript.SystemScriptSource.SystemScriptGraph.NiagaraNodeAssignment_0.SetVariables_#1
  &SystemState
  &self:BurstOnForwardDot
  &FN_Integrator
 ]
}
emitters {
 AccelerationGlow {
  emitterUpdateScript [
   &EmitterState()
   &SpawnRate(
    Spawn Probability=&FN_FrontEndFloatModifier(Gameplay=$User.ForwardDot, Front End=1)
    SpawnRate=21.3333
    Use Spawn Probability=true
   ) {enabled=false}
   &SpawnRate(SpawnRate=21, Use Spawn Probability=false) {module=SpawnRate001}
  ]
  particleSpawnScript [
   /Niagara/Modules/Spawn/Initialization/InitializeParticle(
    Color=&MultiplyLinearColor(
     Linear Color=&LerpLinearColors(
      StartColor=<rgba 0.1609 2.2895 8 0.005>
      EndColor=<rgba 1.4471 2.8875 4 0.0035>
      LerpFactor=&FloatFromCurve(
       FloatCurve=DataInterface() {propertyValues {Curve {keys [
           {arriveTangent=-0.0022, interpMode=RCIM_Cubic, leaveTangent=-0.0022, tangentMode=RCTM_User, time=0, value=0}
           {arriveTangent=0.7451, interpMode=RCIM_Cubic, leaveTangent=0.7451, tangentMode=RCTM_User, time=0.176, value=0.0582}
           {arriveTangent=0.0275, interpMode=RCIM_Cubic, leaveTangent=0.0275, tangentMode=RCTM_User, time=0.4, value=1}
          ]}}}
       CurveIndex=$Emitter.NormalizedLoopAge
      )
     )
     Scale Factor=&Vector4FromCurve(
      CurveIndex=$Emitter.NormalizedLoopAge
      Vector4Curve=DataInterface() {propertyValues {
        AlphaCurve {keys [<at 0.5171 1>,<at 1 0>]}
        BlueCurve {keys [<at 0 1>,<at 1 1>]}
        GreenCurve {keys [<at 0 1>,<at 1 1>]}
        RedCurve {keys [<at 0 1>,<at 1 1>]}
       }}
     )
    )
    Lifetime=0.5
    Module.Write Color=true
    Module.Write Lifetime=true
    Module.Write Mass=true
    Module.Write Material Random=true
    Module.Write Position=false
    Module.Write Scale=false
    Module.Write SpriteRotation=false
    Module.Write SpriteSize=true
    Position=&SimulationPosition()
    Sprite Size=&Multiply_Vector2(
     A=<1 1>
     B=&Vector2DFromFloat(Value=&RandomRangeFloat(Minimum=60, Maximum=65))
    )
   )
   &DynamicMaterialParameters(
    Index 0 Param 1=100
    Module.Param0WriteEnabled=true
    Module.Param1WriteEnabled=false
    Module.Param2WriteEnabled=false
    Module.Param3WriteEnabled=false
   )
   &CameraOffset(Camera Offset Amount=30)
  ]
  particleUpdateScript [
   &ParticleState()
   &SkeletalMeshLocation(
    Inherit Velocity=false
    Interpolate Mesh Sampling=false
    MaskSpawnGroup=false
    Mesh Sampling Type=::NewEnumerator5
    Sampled Position Offset=&RandomRangeVector(Minimum=<-6 15 -6>, Maximum=<6 15 6>)
    Sampled Position Offset Coordinate Space:ENiagaraMeshTransforms=::NewEnumerator3
    Skeletal Mesh=$User.Skeletal Mesh
    Socket Sampling Mode=::NewEnumerator5
   )
   &self:ScratchModule() {module=ScratchModule}
   &ScaleColor(
    Scale Alpha=&FloatFromCurve(
     FloatCurve=DataInterface() {propertyValues {Curve {keys [<at 0 1>,<at 1 0.25>]}}}
     CurveIndex=&ClampFloat(Float=&Abs_Float(Float:NiagaraFloat=&Multiply_Float(A=$Particles.ZDiffValue, B=0.0023)))
    )
    ScaleA=true
    ScaleRGB=true
   )
   &ScaleSpriteSize(Scale Factor=&Vector2DFromFloat(Value=&Lerp_Float(
      A=1
      B=1.4
      Alpha=&FloatFromCurve(FloatCurve=DataInterface() {propertyValues {Curve {keys [<at 0 0>,<at 1 1>]}}}, CurveIndex=$User.ForwardDot)
     )))
   &Color(Color=&ColorFromCurve(
     ColorCurve=DataInterface() {propertyValues {
       AlphaCurve {keys [<at 0 1 RCIM_Cubic>]}
       BlueCurve {keys [<at -1 0>,<at -0.4 0>,<at 0.4 0.0546>,<at 1 1 RCIM_Cubic>]}
       GreenCurve {keys [<at -1 0>,<at -0.4 0.9106>,<at 0.4 1>,<at 1 0.0111 RCIM_Cubic>]}
       RedCurve {keys [<at -1 1>,<at -0.4 1>,<at 0.4 0>,<at 1 0 RCIM_Cubic>]}
      }}
     CurveIndex=&Abs_Float(Float:NiagaraFloat=&Multiply_Float(A=$Particles.ZDiffValue, B=0.0023))
    )) {enabled=false}
  ]
  rendererClasses [&NiagaraSpriteRendererProperties]
  renderers [{rendererClass=&NiagaraSpriteRendererProperties}]
 }
 Flame_Mesh {
  emitterUpdateScript [
   &EmitterState()
   &SpawnRate(
    SpawnRate=&FloatFromCurve(FloatCurve=DataInterface() {propertyValues {Curve {keys [<at 0 20 RCIM_Cubic>,<at 1 0>]}}}, CurveIndex=$Emitter.NormalizedLoopAge)
    Use Spawn Probability=false
   )
   &SpawnRate(Spawn Probability=$System.BurstOnOffFire, SpawnRate=12, Use Spawn Probability=true) {module=SpawnRate001}
  ]
  particleSpawnScript [
   &InitializeParticle(
    Lifetime=&Lerp_Float(A=0.3, B=0.1, Alpha=&FN_FrontEndFloatModifier(Gameplay=$User.ForwardDot, Front End=1))
    Mesh Scale=&Multiply_Vector(
     A=&Multiply_Vector(
      A=&RandomRangeVector(Minimum=<1.25 1.25 1>, Maximum=<1 1 1.5>)
      B=&VectorFromFloat(Value=0.165)
     )
     B=&VectorFromCurve(
      VectorCurve=DataInterface() {propertyValues {
        XCurve {keys [{arriveTangent=-1.9484, interpMode=RCIM_Cubic, leaveTangent=-1.9484, tangentMode=RCTM_User, time=0.1868, value=0.988}, <at 0.6486 0.4398>]}
        YCurve {keys [{arriveTangent=-1.9654, interpMode=RCIM_Cubic, leaveTangent=-1.9654, tangentMode=RCTM_User, time=0.1739, value=0.9808}, <at 0.6611 0.447>]}
        ZCurve {keys [
          {arriveTangent=-0.8115, interpMode=RCIM_Cubic, leaveTangent=-0.8115, tangentMode=RCTM_User, time=0.3, value=1}
          {arriveTangent=-0.068, interpMode=RCIM_Cubic, leaveTangent=-0.068, tangentMode=RCTM_User, time=0.7352, value=0.4504}
         ]}
       }}
      CurveIndex=$Emitter.NormalizedLoopAge
     )
    )
    Module.Write Color=true
    Module.Write Lifetime=true
    Module.Write Mass=false
    Module.Write Material Random=false
    Module.Write Position=false
    Module.Write Scale=true
    Module.Write SpriteRotation=false
    Module.Write SpriteSize=false
    Position=&SimulationPosition()
    Sprite Size=<5 5>
   )
   &DynamicMaterialParameters(
    Index 0 Param 1=&RandomRangeFloat(Minimum=0, Maximum=1)
    Index 0 Param 4=&Lerp_Float(A=2, B=3, Alpha=&FN_FrontEndFloatModifier(Gameplay=$User.ForwardDot, Front End=1))
    Module.Param0WriteEnabled=true
    Module.Param1WriteEnabled=false
    Module.Param2WriteEnabled=false
    Module.Param3WriteEnabled=true
   )
   &InitialMeshOrientation(
    Orientation Axis=<0 -1 0>
    Orientation Vector=<0 0 1>
    Rotation=&RandomRangeVector(Minimum=<0 0 0>, Maximum=<0 0 1>)
    Use Orientation Vector=true
    Use Rotation Vector=true
   ) {module=InitialMeshOrientation001}
  ]
  particleUpdateScript [
   &ParticleState()
   &SkeletalMeshLocation(
    Apply Sampled Orientation to Mesh Particles=true
    Inherit Velocity=false
    Interpolate Mesh Sampling=false
    MaskSpawnGroup=false
    Mesh Sampling Type=::NewEnumerator5
    Sampled Position Offset Coordinate Space:ENiagaraMeshTransforms=::NewEnumerator3
    Skeletal Mesh=$User.Skeletal Mesh
    Socket Sampling Mode=::NewEnumerator5
   )
   &ScaleMeshSize(Scale Factor=&VectorFromCurve(
     VectorCurve=DataInterface() {propertyValues {
       XCurve {keys [<at 0 1>,<at 1 1>]}
       YCurve {keys [<at 0 1>,<at 1 1>]}
       ZCurve {keys [<at 0.9 1 RCIM_Cubic>, {arriveTangent=-4.873, interpMode=RCIM_Cubic, leaveTangent=-4.873, tangentMode=RCTM_User, time=1, value=0}]}
      }}
     Scale Curve=&Lerp_Vector(
      A=<1 1 1>
      B=<1 1 2.5>
      Alpha=&self:InterpolateFloat(Float to Interpolate:NiagaraFloat=$System.BurstOnOff, Initial Value:NiagaraFloat=1, Rate Of Change:NiagaraFloat=20)
     )
    ))
   &ScaleColor(
    Scale Alpha=&Lerp_Float(
     A=0
     B=&Lerp_Float(A=2, B=4, Alpha=&FN_FrontEndFloatModifier(Gameplay=$User.ForwardDot, Front End=1))
     Alpha=$Particles.NormalizedAge
    )
    Scale RGB=&Lerp_Vector(A=<1 1 1>, B=<4 5 5>, Alpha=&FN_FrontEndFloatModifier(Gameplay=$User.ForwardDot, Front End=1))
    ScaleA=true
    ScaleRGB=true
   )
  ]
  rendererClasses [&NiagaraMeshRendererProperties]
  renderers [
   {rendererClass=&NiagaraMeshRendererProperties, rendererIndex=0}
   {rendererClass=&NiagaraMeshRendererProperties, rendererIndex=1}
  ]
 }
 Light {
  emitterUpdateScript [
   &EmitterState()
   &SpawnRate(SpawnRate=4, Use Spawn Probability=false) {enabled=false}
   &SpawnBurst_Instantaneous(
    Spawn Count=1
    Spawn Probability=&MakeCustomFloatFromBool(
     Bool=&SetBoolByIntComparison(A=$Emitter.LoopCount, B=0, Comparison Type=::NewEnumerator0)
     True Float=0
     False Float=1
    )
    Use Spawn Probability=true
   )
  ]
  particleSpawnScript [
   &InitializeParticle(
    Color=<rgba 20 6.5 2 0.5>
    Module.Write Color=true
    Module.Write Lifetime=true
    Module.Write Mass=true
    Module.Write Material Random=true
    Module.Write Position=false
    Module.Write Scale=false
    Module.Write SpriteRotation=false
    Module.Write SpriteSize=true
    Position=&SimulationPosition()
    Sprite Size=<4 4>
   )
   &CameraOffset(Camera Offset Amount=3)
  ]
  particleUpdateScript [
   &ParticleState(Kill Particles When Lifetime Has Elapsed=false, Let Infinitely Lived Particles Die When Emitter Deactivates=true)
   &SkeletalMeshLocation(
    Inherit Velocity=false
    Interpolate Mesh Sampling=false
    MaskSpawnGroup=false
    Mesh Sampling Type=::NewEnumerator5
    Sampled Position Offset=<0 5 0>
    Sampled Position Offset Coordinate Space:ENiagaraMeshTransforms=::NewEnumerator3
    Skeletal Mesh=$User.Skeletal Mesh
    Socket Sampling Mode=::NewEnumerator5
   )
   &FN_Light_v3(
    Color=&MultiplyLinearColor(
     Linear Color=<rgba 19.3409 123.3575 300 1>
     Scale Factor=&MakeVector4_FromFloat(Float=&FloatFromCurve(FloatCurve=DataInterface() {propertyValues {Curve {keys [<at 0.5 1 RCIM_Cubic>,<at 1 0>]}}}, CurveIndex=$Emitter.NormalizedLoopAge))
    )
    Create Color Attribute=true
    Create Exponent Attribute=true
    Create Radius Attribute=true
    Exponent:NiagaraFloat=&Lerp_Float(A=5, B=2, Alpha=&MakeFloatFromInt(Int=0))
    Radius:NiagaraFloat=&self:InterpolateFloat(
     Float to Interpolate:NiagaraFloat=&RandomRangeFloat(
      Minimum=30
      Maximum=40
      Evaluation Type=::NewEnumerator1
      Randomness Mode=::NewEnumerator1
      Random Seed=&Add_Integer(A=$Particles.SpawnGroup, B=&Add_Integer(A=$Emitter.LoopCount, B=$System.LoopCount))
      Fixed Random Seed=true
     )
     Initial Value:NiagaraFloat=40
     Rate Of Change:NiagaraFloat=0
    )
    Scale Color:NiagaraFloat=&self:InterpolateFloat(
     Float to Interpolate:NiagaraFloat=&RandomRangeFloat(Minimum=0.2, Maximum=1, Evaluation Type=::NewEnumerator1)
     Initial Value:NiagaraFloat=1
     Rate Of Change:NiagaraFloat=5
    )
   )
   &FN_Light_v3(
    Color=<rgba 200 65 20 1>
    Create Color Attribute=true
    Create Exponent Attribute=true
    Create Radius Attribute=true
    Enable Particle Light Enable=true
    Exponent=1
    Light Enabled=true
    Radius:NiagaraFloat=&Lerp_Float(
     A=&RandomRangeFloat(
      Minimum=17
      Maximum=20
      Evaluation Type=::NewEnumerator1
      Randomness Mode=::NewEnumerator1
      Random Seed=&Add_Integer(A=$Particles.SpawnGroup, B=&Add_Integer(A=$Emitter.LoopCount, B=$System.LoopCount))
      Fixed Random Seed=true
     )
     B=&RandomRangeFloat(
      Minimum=25
      Maximum=25
      Evaluation Type=::NewEnumerator1
      Randomness Mode=::NewEnumerator1
      Random Seed=&Add_Integer(A=$Particles.SpawnGroup, B=&Add_Integer(A=$Emitter.LoopCount, B=$System.LoopCount))
      Fixed Random Seed=true
     )
     Alpha=&MakeFloatFromInt(Int=$Particles.PlacementPicker)
    )
    Scale Color:NiagaraFloat=&RandomRangeFloat(Minimum=0.25, Maximum=1, Evaluation Type=::NewEnumerator1)
   ) {enabled=false, module=FN_Light_v3001}
  ]
  rendererClasses [&NiagaraSpriteRendererProperties, &NiagaraLightRendererProperties]
  renderers [
   {rendererClass=&NiagaraSpriteRendererProperties}
   {rendererClass=&NiagaraLightRendererProperties, rendererIndex=1}
  ]
 }
 Smoke {
  bEnabled=false
  emitterUpdateScript [
   &EmitterState()
   /SetVariables(
    Emitter.ForwardDotInterp:NiagaraFloat=&self:InterpolateFloat(Float to Interpolate:NiagaraFloat=$User.ForwardDot, Initial Value:NiagaraFloat=0, Rate Of Change:NiagaraFloat=3)
    Emitter.RightDotInterp:NiagaraFloat=&self:InterpolateFloat(Float to Interpolate:NiagaraFloat=$User.RightDot, Initial Value:NiagaraFloat=0, Rate Of Change:NiagaraFloat=3)
   ) {isSetParameters=true}
   &SpawnRate(
    Spawn Probability=&MakeCustomFloatFromBool(Bool=&SetBoolByFloatComparison(A=$User.ForwardDot, B=0.65), True Float=1, False Float=0)
    SpawnRate=150
    Use Spawn Probability=true
   ) {module=SpawnRate001}
   &SpawnRate(
    Spawn Probability=$System.BurstOnOffSmoke
    SpawnRate=&FloatFromCurve(
     FloatCurve=DataInterface() {propertyValues {Curve {keys [<at 0 1>,<at 1 0>]}}}
     CurveIndex=$System.FN_Integrator002.NormalizedValue
     Scale Curve=1000
    )
    Use Spawn Probability=true
   )
  ]
  particleSpawnScript [
   &InitializeParticle(
    Color=&UniformRangedLinearColor(Minimum=<rgba 0.505 0.505 0.505 0.005>, Maximum=<rgba 0.771 0.771 0.771 0.06>)
    Lifetime=&Multiply_Float(A=&RandomRangeFloat(Minimum=0.75, Maximum=1), B=0.25)
    Module.Write Color=true
    Module.Write Lifetime=true
    Module.Write Mass=true
    Module.Write Material Random=true
    Module.Write Position=false
    Module.Write Scale=false
    Module.Write SpriteRotation=true
    Module.Write SpriteSize=true
    Position=&SimulationPosition()
    Sprite Rotation=&UniformRangedFloat(Minimum=0, Maximum=360)
    Sprite Size=&Multiply_Vector2(
     A=&Multiply_Vector2(
      A=&Vector2DFromFloat(Value=&UniformRangedFloat(Minimum=0.5, Maximum=1))
      B=&Vector2DFromFloat(Value=40)
     )
     B=&Vector2DFromFloat(Value=&FloatFromCurve(FloatCurve=DataInterface() {propertyValues {Curve {keys [<at 0 2>,<at 1 1.5>]}}}, CurveIndex=$System.FN_Integrator002.NormalizedValue))
    )
   )
   &ApplyInitialForces()
   /SetVariables(Particles.PickBone:NiagaraInt32=&RandomRangeInt(Minimum=0, Maximum=1)) {enabled=false, isSetParameters=true}
   &SkeletalMeshLocation(
    Inherit Velocity=false
    MaskSpawnGroup=false
    Mesh Sampling Type=::NewEnumerator5
    Sampled Position Offset Coordinate Space:ENiagaraMeshTransforms=::NewEnumerator3
    Skeletal Mesh=$User.Skeletal Mesh
    Socket Sampling Mode=::NewEnumerator5
   ) {enabled=false}
   &SkeletalMeshLocation(
    Inherit Velocity=false
    Interpolate Mesh Sampling=false
    MaskSpawnGroup=false
    Mesh Sampling Type=::NewEnumerator5
    Sampled Position Offset Coordinate Space:ENiagaraMeshTransforms=::NewEnumerator3
    Skeletal Mesh=$User.Skeletal Mesh
    Socket Sampling Mode=::NewEnumerator5
   ) {module=SkeletalMeshLocation001}
   &AddVelocity(Velocity=<0 1000 0>)
  ]
  particleUpdateScript [
   &InheritVelocity(
    Module.Limit Speed=false
    Velocity Scale=&Multiply_VectorByFloat(Vector=<0.1 0.1 0>, Float=&Lerp_Float(A=1, B=0, Alpha=$Particles.NormalizedAge))
   ) {enabled=false}
   &SubUVAnimation(
    SubUV Animation Mode:ENiagara_SubUVLookupMode=::NewEnumerator3
    SubUV Start Time Offset:NiagaraFloat=&RandomRangeFloat(Minimum=0, Maximum=1)
   )
   &ParticleState()
   &SolveForcesAndVelocity(Clamp Velocity=false, Limit Acceleration=false)
   &ScaleSpriteSize(Scale Factor=&Vector2DFromFloat(Value=&FloatFromCurve(FloatCurve=DataInterface() {propertyValues {Curve {keys [
          {arriveTangent=3.1801, interpMode=RCIM_Cubic, leaveTangent=3.1801, tangentMode=RCTM_User, time=0, value=0}
          {interpMode=RCIM_Cubic, tangentMode=RCTM_User, time=0.5164, value=1}
          {arriveTangent=-3.6914, interpMode=RCIM_Cubic, leaveTangent=-3.6914, tangentMode=RCTM_User, time=1, value=0}
         ]}}}))) {enabled=false}
   /SetVariables(Particles.SubImageIndex=&FloatFromCurve(FloatCurve=DataInterface() {propertyValues {Curve {keys [
         {arriveTangent=29.8623, interpMode=RCIM_Cubic, leaveTangent=29.8623, tangentMode=RCTM_User, time=-0.0046, value=5.0556}
         <at 1.0011 22.1689 RCIM_Cubic>
        ]}}})) {isSetParameters=true}
   /SetVariables(Particles.SpriteDynaMatParams:Vector4f=&MakeVector4(
     X=&FloatFromCurve(FloatCurve=DataInterface() {propertyValues {Curve {keys [<at -0.0046 1.0113 RCIM_Cubic>,<at 1.0011 0.002>]}}})
     Y=0
     Z=0
     W=&FloatFromCurve(FloatCurve=DataInterface() {propertyValues {Curve {keys [<at 0 0 RCIM_Cubic>,<at 1 0.5>]}}})
    )) {isSetParameters=true}
   &ScaleColor(
    Scale Alpha=&FloatFromCurve(FloatCurve=DataInterface() {propertyValues {Curve {keys [
         {arriveTangent=-3.1757, interpMode=RCIM_Cubic, leaveTangent=-3.1757, tangentMode=RCTM_User, time=0.5, value=1.0113}
         {arriveTangent=-1.5121, interpMode=RCIM_Cubic, leaveTangent=-1.5121, tangentMode=RCTM_User, time=1.0011, value=0.002}
        ]}}})
    ScaleA=true
    ScaleRGB=false
   )
  ]
  rendererClasses [&NiagaraSpriteRendererProperties]
  renderers [
   {rendererClass=&NiagaraSpriteRendererProperties}
   {rendererClass=&NiagaraSpriteRendererProperties, rendererIndex=1}
  ]
 }
 Sparks {
  bEnabled=false
  emitterUpdateScript [
   &EmitterState()
   &SpawnRate(
    SpawnRate=&Multiply_Float(
     A=2000
     B=&FloatFromCurve(
      FloatCurve=DataInterface() {propertyValues {Curve {keys [<at 0 1>,<at 0.178 0.0307>,<at 0.3989 1>,<at 1 0>]}}}
      CurveIndex=$Emitter.NormalizedLoopAge
     )
    )
    Use Spawn Probability=true
   )
   &SpawnRate(
    Spawn Probability=&MakeCustomFloatFromBool(Bool=&SetBoolByFloatComparison(A=$User.ForwardDot, B=0.65), True Float=1, False Float=0)
    SpawnRate=50
    Use Spawn Probability=true
   ) {enabled=false, module=SpawnRate002}
   &SpawnRate(
    Spawn Probability=$System.BurstOnOffSparks
    SpawnRate=&FloatFromCurve(
     FloatCurve=DataInterface() {propertyValues {Curve {keys [
         {arriveTangent=-13.0128, interpMode=RCIM_Cubic, leaveTangent=-13.0128, tangentMode=RCTM_User, time=0, value=1}
         {arriveTangent=-1.0133, interpMode=RCIM_Cubic, leaveTangent=-1.0133, tangentMode=RCTM_User, time=0.1027, value=0.3148}
         <at 1 0>
        ]}}}
     CurveIndex=$System.FN_Integrator.Value
     Scale Curve=500
    )
    Use Spawn Probability=true
   ) {enabled=false, module=SpawnRate003}
  ]
  particleSpawnScript [
   &InitializeParticle(
    Color=<rgba 5 2.625 1.25 1>
    Lifetime=&Multiply_Float(
     A=&RandomRangeFloat(Minimum=0.1, Maximum=0.2)
     B=&FloatFromCurve(
      FloatCurve=DataInterface() {propertyValues {Curve {keys [{arriveTangent=-1.6119, interpMode=RCIM_Cubic, leaveTangent=-1.6119, tangentMode=RCTM_User, time=0, value=1}, <at 1 0.5>]}}}
      CurveIndex=$Emitter.NormalizedLoopAge
      Scale Curve=5
     )
    )
    Module.Write Color=true
    Module.Write Lifetime=true
    Module.Write Mass=true
    Module.Write Material Random=true
    Module.Write Position=false
    Module.Write Scale=false
    Module.Write SpriteRotation=false
    Module.Write SpriteSize=true
    Position=&SimulationPosition()
    Sprite Size=&Lerp_Vector2(
     A=&UniformRangedVector2D(Minimum=<0.5 1>, Maximum=<1 2>)
     B=&UniformRangedVector2D(Minimum=<1 1.5>, Maximum=<2 4>)
     Alpha=&FN_FrontEndFloatModifier(Gameplay=$User.ForwardDot, Front End=1)
    )
   )
   &SkeletalMeshLocation(
    Inherit Velocity=false
    Interpolate Mesh Sampling=false
    MaskSpawnGroup=false
    Mesh Sampling Type=::NewEnumerator5
    Sampled Position Offset=&Lerp_Vector(A=$Particles.StartPosition, B=$Particles.EndPosition, Alpha=$Particles.NormalizedAge)
    Sampled Position Offset Coordinate Space:ENiagaraMeshTransforms=::NewEnumerator3
    Skeletal Mesh=$User.Skeletal Mesh
    Socket Sampling Mode=::NewEnumerator5
    Use Sampled Mesh Attribute as Ribbon ID=false
   ) {module=SkeletalMeshLocation001}
   &AddVelocity(Velocity=<0 500 0>)
   &CylinderLocation(
    Cylinder Height=0
    Cylinder Radius=&Multiply_Float(
     A=25
     B=&FloatFromCurve(
      FloatCurve=DataInterface() {propertyValues {Curve {keys [
          <at 0 0.25 RCIM_Cubic>
          <at 0.1335 1>
          {arriveTangent=-5.3735, interpMode=RCIM_Cubic, leaveTangent=-5.3735, tangentMode=RCTM_User, time=0.36, value=1}
          <at 0.65 0.5>
         ]}}}
      CurveIndex=$Emitter.NormalizedLoopAge
     )
    )
    MaskSpawnGroup=false
    Module.Enable Lathe Profile=false
    Module.Override Randomness=false
    Module.Surface Only=false
    Orientation Axis:ENiagaraOrientationAxis=::YAxis
   )
  ]
  particleUpdateScript [
   &SolveForcesAndVelocity(Clamp Velocity=false, Limit Acceleration=false)
   &ParticleState()
   &ScaleColor(
    Scale Alpha=&FloatFromCurve(
     FloatCurve=DataInterface() {propertyValues {Curve {keys [<at 0 1>,<at 1 0>]}}}
     CurveIndex=&Multiply_Float(A=$Particles.NormalizedAge, B=2)
    )
    Scale RGB=&MakeVectorFromLinearColorRGB(Color=&ColorFromCurve(
      ColorCurve=DataInterface() {propertyValues {
        AlphaCurve {keys []}
        BlueCurve {keys [<at 0 0.25 RCIM_Cubic>,<at 0.5 0.5>,<at 1 0.875 RCIM_Cubic>]}
        GreenCurve {keys [<at 0 0.525 RCIM_Cubic>,<at 0.5 0.6833>,<at 1 0.9208 RCIM_Cubic>]}
        RedCurve {keys [<at 0 1 RCIM_Cubic>,<at 0.5 1>,<at 1 1 RCIM_Cubic>]}
       }}
      CurveIndex=&Multiply_Float(A=$Particles.NormalizedAge, B=2)
     ))
    ScaleA=true
    ScaleRGB=true
   )
   &ScaleColor(
    Scale RGB=&VectorFromFloat(Value=&FloatFromCurve(
      FloatCurve=DataInterface() {propertyValues {Curve {keys [
          <at 0 100 RCIM_Cubic>
          {arriveTangent=-36.3579, interpMode=RCIM_Cubic, leaveTangent=-36.3595, tangentMode=RCTM_User, time=0.25, value=10}
          <at 1 1 RCIM_Cubic>
         ]}}}
      CurveIndex=&Multiply_Float(A=$Particles.NormalizedAge, B=2)
     ))
    ScaleA=true
    ScaleRGB=true
   ) {module=ScaleColor001}
   &ScaleSpriteSize(Scale Factor=&Vector2DFromCurve(
     Vector2Curve=DataInterface() {propertyValues {
       XCurve {keys [<at 0 1>,<at 1 0>]}
       YCurve {keys [{arriveTangent=18.2053, interpMode=RCIM_Cubic, leaveTangent=18.2053, tangentMode=RCTM_User, time=0, value=1}, <at 1 10>]}
      }}
     CurveIndex=&Multiply_Float(A=$Particles.NormalizedAge, B=2)
     ScaleCurve=<1 1>
    ))
  ]
  rendererClasses [&NiagaraSpriteRendererProperties]
  renderers [{rendererClass=&NiagaraSpriteRendererProperties}]
 }
 Sparks001 {
  emitterUpdateScript [
   &EmitterState()
   &SpawnRate(
    SpawnRate=&Multiply_Float(
     A=250
     B=&FloatFromCurve(
      FloatCurve=DataInterface() {propertyValues {Curve {keys [<at 0 0.25>, {arriveTangent=-0.6607, interpMode=RCIM_Cubic, leaveTangent=-0.6607, tangentMode=RCTM_User, time=0.85, value=1}, <at 1 0>]}}}
      CurveIndex=$Emitter.NormalizedLoopAge
     )
    )
    Use Spawn Probability=true
   )
  ]
  particleSpawnScript [
   &InitializeParticle(
    Color=<rgba 1.6456 3.5593 5 1>
    Lifetime=&Multiply_Float(A=0.1, B=&RandomRangeFloat(Minimum=0.5, Maximum=1))
    Module.Write Color=true
    Module.Write Lifetime=true
    Module.Write Mass=true
    Module.Write Material Random=true
    Module.Write Position=false
    Module.Write Scale=false
    Module.Write SpriteRotation=false
    Module.Write SpriteSize=true
    Position=&SimulationPosition()
    Sprite Size=&Multiply_Vector2(
     A=&RandomRangeVector2D(Minimum=<0 0>, Maximum=<2 4>)
     B=&Vector2DFromFloat(Value=&FloatFromCurve(
       FloatCurve=DataInterface() {propertyValues {Curve {keys [{arriveTangent=-1.8584, interpMode=RCIM_Cubic, leaveTangent=-1.8584, tangentMode=RCTM_User, time=0, value=2}, <at 1 0.5>]}}}
       CurveIndex=$Emitter.NormalizedLoopAge
      ))
    )
   )
   &AddVelocity(Velocity=<500 -500 500>)
  ]
  particleUpdateScript [
   &SolveForcesAndVelocity(Clamp Velocity=false, Limit Acceleration=false)
   &ParticleState()
   /SetVariables(Particles.StartPosition:Vector3f=&Multiply_VectorByFloat(
     Vector=&RandomVector(Vector Scale=30)
     Float=&FloatFromCurve(
      FloatCurve=DataInterface() {propertyValues {Curve {keys [
          <at 0 0.25>
          {arriveTangent=-1.2237, interpMode=RCIM_Cubic, leaveTangent=-1.2237, tangentMode=RCTM_User, time=0.0397, value=1}
          {arriveTangent=-0.6607, interpMode=RCIM_Cubic, leaveTangent=-0.6607, tangentMode=RCTM_User, time=1, value=0.25}
         ]}}}
      CurveIndex=$Emitter.NormalizedLoopAge
     )
    )) {isSetParameters=true}
   /SetVariables(Particles.EndPosition:Vector3f=&Add_Vector(
     A=&Multiply_Vector(A=$Particles.StartPosition, B=<1.5 1 1.5>)
     B=&Multiply_Vector(
      A=&RandomRangeVector(Minimum=<0 50 0>, Maximum=<0 250 0>)
      B=&VectorFromFloat(Value=&FloatFromCurve(
        FloatCurve=DataInterface() {propertyValues {Curve {keys [{arriveTangent=-1.6119, interpMode=RCIM_Cubic, leaveTangent=-1.6119, tangentMode=RCTM_User, time=0, value=1}, <at 1 0.5>]}}}
        CurveIndex=$System.FN_Integrator.NormalizedValue
        Scale Curve=2
       ))
     )
    )) {isSetParameters=true}
   /SetVariables(Particles.SpriteAlignment=&FN_FrontEndVector3Modifier(
     Gameplay=&Subtract_Vector(A=$Particles.EndPosition, B=$Particles.StartPosition)
     Front End=/Game/Effects/Niagara/DynamicInputs/FrontEnd/FN_FrontEndRotateVector(
      Angle:NiagaraFloat=0.2
      Vector to Rotate:Vector3f=&Subtract_Vector(A=$Particles.EndPosition, B=$Particles.StartPosition)
     )
    )) {isSetParameters=true}
   &SkeletalMeshLocation(
    Inherit Velocity=false
    Interpolate Mesh Sampling=false
    MaskSpawnGroup=false
    Mesh Sampling Type=::NewEnumerator5
    Sampled Position Offset=&Lerp_Vector(A=$Particles.StartPosition, B=$Particles.EndPosition, Alpha=$Particles.NormalizedAge)
    Sampled Position Offset Coordinate Space:ENiagaraMeshTransforms=::NewEnumerator3
    Skeletal Mesh=$User.Skeletal Mesh
    Socket Sampling Mode=::NewEnumerator5
    Use Sampled Mesh Attribute as Ribbon ID=false
   ) {module=SkeletalMeshLocation001}
   &ScaleColor(
    Scale Alpha=&FloatFromCurve(
     FloatCurve=DataInterface() {propertyValues {Curve {keys [<at 0 1>,<at 1 0>]}}}
     CurveIndex=&Multiply_Float(A=$Particles.NormalizedAge, B=2)
    )
    Scale RGB=&MakeVectorFromLinearColorRGB(Color=&ColorFromCurve(
      ColorCurve=DataInterface() {propertyValues {
        AlphaCurve {keys []}
        BlueCurve {keys [<at -0.0049 1>,<at 0.4971 1 RCIM_Cubic>,<at 1 0.875 RCIM_Cubic>]}
        GreenCurve {keys [<at -0.0049 0.6566>,<at 0.4971 0.88 RCIM_Cubic>,<at 1 0.9208 RCIM_Cubic>]}
        RedCurve {keys [<at -0.0049 0.3049>,<at 0.4971 0.2705 RCIM_Cubic>,<at 1 1 RCIM_Cubic>]}
       }}
      CurveIndex=&Multiply_Float(A=$Particles.NormalizedAge, B=2)
     ))
    ScaleA=true
    ScaleRGB=true
   )
   &ScaleColor(
    Scale RGB=&VectorFromFloat(Value=&FloatFromCurve(
      FloatCurve=DataInterface() {propertyValues {Curve {keys [
          <at 0 100 RCIM_Cubic>
          {arriveTangent=-36.3579, interpMode=RCIM_Cubic, leaveTangent=-36.3595, tangentMode=RCTM_User, time=0.25, value=10}
          <at 1 1 RCIM_Cubic>
         ]}}}
      CurveIndex=&Multiply_Float(A=$Particles.NormalizedAge, B=2)
     ))
    ScaleA=true
    ScaleRGB=true
   ) {module=ScaleColor001}
   &ScaleSpriteSize(Scale Factor=&Vector2DFromCurve(
     Vector2Curve=DataInterface() {propertyValues {
       XCurve {keys [<at 0 1>,<at 1 0>]}
       YCurve {keys [{arriveTangent=18.2053, interpMode=RCIM_Cubic, leaveTangent=18.2053, tangentMode=RCTM_User, time=0, value=1}, <at 1 10>]}
      }}
     CurveIndex=&Multiply_Float(A=$Particles.NormalizedAge, B=2)
     ScaleCurve=<1 1>
    ))
   /SetVariables(
    Particles.ColorAlt:LinearColor=<rgba 0 0 0 0.015>
    Particles.PositionAlt:Vector3f=&Lerp_Vector(A=$Output.SkeletalMeshLocation001.SampledPosition, B=Linked() {linkedVariable {name=Particles.Position, type=NiagaraPosition}}, Alpha=0.4)
    Particles.SpriteFacingAlt:Vector3f=<0 0 1>
    Particles.SpriteRotationAlt:NiagaraFloat=&RandomRangeFloat(Minimum=-360, Maximum=360)
    Particles.SpriteSizeAlt:Vector2f=&Vector2DFromFloat(Value=&Multiply_Float(
      A=&RandomRangeFloat(Minimum=60, Maximum=90)
      B=&Lerp_Float(
       A=1
       B=&RandomRangeFloat(Minimum=0, Maximum=1)
       Alpha=&MakeFloatFromInt(Int=$Particles.SpawnGroup)
      )
     ))
   ) {isSetParameters=true}
  ]
  rendererClasses [&NiagaraSpriteRendererProperties]
  renderers [{rendererClass=&NiagaraSpriteRendererProperties}]
 }
}
summary {
 systemName=NS_Glider_Combat_Doll_Deploy
 emitters [
  {bEnabled=false, emitterName=Sparks, rendererClasses [&NiagaraSpriteRendererProperties]}
  {emitterName=Flame_Mesh, rendererClasses [&NiagaraMeshRendererProperties]}
  {emitterName=Light, rendererClasses [&NiagaraSpriteRendererProperties, &NiagaraLightRendererProperties]}
  {emitterName=AccelerationGlow, rendererClasses [&NiagaraSpriteRendererProperties]}
  {bEnabled=false, emitterName=Smoke, rendererClasses [&NiagaraSpriteRendererProperties]}
  {emitterName=Sparks001, rendererClasses [&NiagaraSpriteRendererProperties]}
 ]
 userVariables [
  User.bIsFrontEnd {type=NiagaraBool}
  User.bIsFrontEndPreview {type=NiagaraBool}
  User.ForwardDot {type=NiagaraFloat, defaultValue=0, description=""}
  User.RightDot {type=NiagaraFloat, defaultValue=0, description=""}
  User.Skeletal Mesh {
   type=NiagaraDataInterfaceSkeletalMesh
   defaultValue {
    struct=NiagaraExt_VariableValue_DataInterface
    value {dataInterface=&self:NiagaraDataInterfaceSkeletalMesh_0}
   }
  }
  User.TheOtherForwardDot {type=NiagaraFloat, defaultValue=0, description=""}
 ]
}
