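# Niagara trail effect for the Combat Doll glider chute, drawing twin mesh trails attached to the glider's skeletal mesh.
# One persistent particle samples two bones or sockets, deriving offset, rotated transforms for the two mesh renderers.
# System-level exhaust, trail length and opacity drive the dynamic material, length update and alpha.
# Front-end modifiers swap these for exhaust from local sampled velocity and interpolated length and opacity.
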NiagaraSystem /BRCosmetics/Gadgets/Assets/VinderTech_GliderChute/Glider_Combat_Doll/FX/NS_Glider_Combat_Doll_Trail
dependencies {
 usedDataInterfaces []
 usedDynamicInputs []
 usedModules [
  &SystemState
  &self:Set_Current_Exhaust_01
  &self:Set_Trail_Length_01
  &self:Set_Trail_Opacity_01
 ]
}
emitters {MeshTrails {
  emitterUpdateScript [
   &EmitterState()
   &SpawnBurst_Instantaneous(Spawn Count=HlslExpression() {hlslExpression="Emitter.LoopCount == 0"}, Use Spawn Probability=false) {module=SpawnBurst_Instantaneous001}
  ]
  particleSpawnScript [
   /Niagara/Modules/Spawn/Initialization/InitializeParticle(
    Color=<rgba 1 1 1 0.1>
    Module.Write Color=true
    Module.Write Lifetime=true
    Module.Write Mass=false
    Module.Write Material Random=true
    Module.Write Position=false
    Module.Write Scale=false
    Module.Write SpriteRotation=false
    Module.Write SpriteSize=false
   )
   &self:Mesh_Trail_Setup(
    Trail Length:NiagaraFloat=&FN_FrontEndFloatModifier(Gameplay=150, Front End=200)
    Trail Width=150
   ) {module=Mesh_Trail_Setup}
  ]
  particleUpdateScript [
   &ParticleState(Kill Particles When Lifetime Has Elapsed=false, Let Infinitely Lived Particles Die When Emitter Deactivates=true)
   &SkeletalMeshLocation(
    Apply Sampled Orientation to Mesh Particles=true
    Apply To Particle Position=false
    Bone Sampling Mode=::NewEnumerator4
    Inherit Velocity=false
    Interpolate Mesh Sampling=false
    MaskSpawnGroup=false
    Sampled Position Offset Coordinate Space:ENiagaraMeshTransforms=::NewEnumerator3
    Skeletal Mesh=$User.SkeletalMesh
    Write Sampled Mesh Attrributes to Particle Attributes=true
   )
   &SkeletalMeshLocation(
    Apply Sampled Orientation to Mesh Particles=true
    Apply To Particle Position=false
    Bone / Socket Index=1
    Bone Sampling Mode=::NewEnumerator4
    Inherit Velocity=false
    Interpolate Mesh Sampling=false
    MaskSpawnGroup=false
    Sampled Position Offset Coordinate Space:ENiagaraMeshTransforms=::NewEnumerator3
    Skeletal Mesh=$User.SkeletalMesh
    Write Sampled Mesh Attrributes to Particle Attributes=true
   ) {module=SkeletalMeshLocation001}
   /SetVariables(
    Particles.TrailOrientation:Quat4f=&MultiplyQuaternion(Quaternion A=$Particles.SkeletalMeshLocation001.SampledRotation, Quaternion B=<0 0 1 1>)
    Particles.TrailPosition:Vector3f=&Add_Vector(
     A=$Particles.SkeletalMeshLocation001.SampledPosition
     B=&RotateVectorByQuaternion(Vector To Rotate=<10 10 -5>, Quaternion=$Particles.SkeletalMeshLocation001.SampledRotation)
    )
    Particles.TrailTwinOrientation:Quat4f=&MultiplyQuaternion(Quaternion A=$Particles.SkeletalMeshLocation.SampledRotation, Quaternion B=<0 0 -1 1>)
    Particles.TrailTwinPosition:Vector3f=&Add_Vector(
     A=$Particles.SkeletalMeshLocation.SampledPosition
     B=&RotateVectorByQuaternion(Vector To Rotate=<-10 10 5>, Quaternion=$Particles.SkeletalMeshLocation.SampledRotation)
    )
   ) {isSetParameters=true}
   &self:Exhaust_to_DynaMat(
    ExhaustVector:Vector3f=&FN_FrontEndVector3Modifier(
     Gameplay=&Multiply_Vector(A=$System.CurrentExhaust, B=<1 3 1>)
     Front End=&self:Front_End_Exhaust(
      Rate Of Change:NiagaraFloat=10
      Velocity Input:Vector3f=&TransformVector(Vector=$Particles.SkeletalMeshLocation.SampledVelocity, Source Space=::World, Destination Space=::Local)
      Velocity Multiplier:Vector3f=<0 -4 0>
     )
    )
    SpeedMult=1
   ) {module=Exhaust_to_DynaMat}
   &self:Mesh_Trail_Length_Update(Dynamic Length Scale:NiagaraFloat=&FN_FrontEndFloatModifier(
     Gameplay=$System.TrailLength
     Front End=&self:InterpolateFloat(Rate Of Change:NiagaraFloat=3, Target Value:NiagaraFloat=1.25)
    )) {module=Mesh_Trail_Length_Update}
   &ScaleColor(
    Scale Alpha=&FN_FrontEndFloatModifier(
     Gameplay=$System.TrailOpacity
     Front End=&self:InterpolateFloat(Rate Of Change:NiagaraFloat=2, Target Value:NiagaraFloat=1)
    )
    ScaleA=true
    ScaleRGB=false
   )
  ]
  rendererClasses [&NiagaraMeshRendererProperties]
  renderers [
   {rendererClass=&NiagaraMeshRendererProperties, rendererIndex=0}
   {rendererClass=&NiagaraMeshRendererProperties, rendererIndex=1}
  ]
 }}
summary {
 systemName=NS_Glider_Combat_Doll_Trail
 emitters [{emitterName=MeshTrails, rendererClasses [&NiagaraMeshRendererProperties]}]
 userVariables [
  User.bIsFrontEnd {type=NiagaraBool}
  User.bIsFrontEndPreview {type=NiagaraBool}
  User.ForwardDot {type=NiagaraFloat, defaultValue=0, description=""}
  User.RightDot {type=NiagaraFloat, defaultValue=0, description=""}
  User.SkeletalMesh {
   type=NiagaraDataInterfaceSkeletalMesh
   defaultValue {
    struct=NiagaraExt_VariableValue_DataInterface
    value {dataInterface=&self:NiagaraDataInterfaceSkeletalMesh_0}
   }
  }
 ]
}
